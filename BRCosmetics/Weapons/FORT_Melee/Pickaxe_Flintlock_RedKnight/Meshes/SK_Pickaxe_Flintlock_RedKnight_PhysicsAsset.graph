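# Physics asset for the Flintlock Red Knight two-handed pickaxe cosmetic's skeletal mesh, providing simplified collision for the weapon.
# A single capsule body on the Melee_TwoHanded_Handle bone, offset and rotated from the bone origin, is the weapon's only collision shape.
# The asset has no constraints, so it defines no jointed or secondary physical parts.
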PhysicsAsset /BRCosmetics/Weapons/FORT_Melee/Pickaxe_Flintlock_RedKnight/Meshes/SK_Pickaxe_Flintlock_RedKnight_PhysicsAsset
bodies {Melee_TwoHanded_Handle [{shapeType=Capsule, length=74.633, radius=38.1795, center=<-0.6617 -5.5568 29.515>, rotation=<rot -0.9068 -90 0.018>}]}
constraints []
